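; Gameplay ability for the Tandem BrawnySmasher NPC that plays a transformation montage and grants the AI a configured item (AGID).
; On activation it commits, waits briefly for the behavior tree to stop, then validates the item, pawn and AI bot controller.
; It fires a transformation-started cue and plays the montage, whose notify triggers a second cue and gives the item to the controller's inventory.
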(event Ability|EventActivateAbility
  (bind _returnvalue (CallFunction|GetAvatarasFortPlayerPawn))
  (Ability|CommitAbility)
  (Utilities|FlowControl|Delay 0.1) ; Delay to give time to BT to be stop
  (Utilities|IsValid (Variables|Default|GetAGIDToApply)
    (:Is%20Valid
      (Utilities|IsValid _returnvalue
        (:Is%20Valid
          (bind _asfort_athena_aibot_controller (Utilities|Casting|CastToFortAthenaAIBotController (Pawn|GetController _returnvalue))
            (:then
              (Ability|ExecuteGameplayCueOnOwner (Variables|Default|GetGameplay%20Cue%20TransformationStarted%20Tag))
              (Ability|Tasks|PlayMontageAndWaitNotify "None" "/Game/Animation/Game/NPCs/BrawnySmasher/Transition_Short_M.Transition_Short_M"
                (:OnCompleted
                  (Utilities|FlowControl|Delay 0.1))
                (:OnInterrupted
                  _)
                (:OnCancelled
                  _)
                (:OnNotifyBegin
                  (Ability|ExecuteGameplayCueOnOwner (Variables|Default|GetGameplay%20Cue%20Tag))
                  (Inventory|GiveItemtoInventoryOwner _asfort_athena_aibot_controller (Variables|Default|GetAGIDToApply))))
              (Ability|EndAbility))
            (:CastFailed
              _)))
        (:Is%20Not%20Valid
          _)))
    (:Is%20Not%20Valid
      _))
  (Ability|EndAbility))

(event Ability|EventOnEndAbility
  (params (bWasCancelled :type Boolean)))
